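; Advanced Storm Beacon V2 device that overrides one storm phase's timing, radius, damage and movement, optionally fixing the end location at its own position.
; GetStormPhase reports the phase it serves to the Storm Caller; GetStormPhaseOverride builds that phase struct, using passed-in values for some negative options.
; Property changes rescale the radius preview mesh and set end-radius visibility on authority.
; MigrateLegacyUserOptions converts old centimeter and damage options to meters and enum levels, then refreshes saved data.
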(event EventOnPropertiesChanged
  (|CallParentFunction)
  (Utilities|FlowControl|Sequence
    (:then_0
      (Variables|Default|SetEndRadiusInCm (* (Variables|UserOptions|Basic|GetStormRadius) 1e+02)))
    (:then_1
      (Networking|SwitchHasAuthority
        (:Authority
          (Visibility|SetVisiblityDuringPhase (Variables|Default|GetEndRadiusVisibleInGameComponent) (select (and (Utilities|Enum|Equal(Enum) (Variables|UserOptions|Basic|GetMovementBehavior) "NewEnumerator1") (> (Variables|Default|GetEndRadiusInCm) 0.0)) "NewEnumerator1" "NewEnumerator0")))))
    (:then_2
      (CallFunction|UpdateEndRadiusMeshWorldScale))))

(event OnVisibilityChanged(VisibleInGameComponent)
  (params (Visible :type Boolean))
  (Audio|Components|Audio|SetVolumeMultiplier (Variables|Default|GetStorm_Manager_Loop_Cue) (select Visible 1.0 0.0) :!class /Script/Engine.AudioComponent))

(event OnMeshandMaterialsLoaded(RadiusPreview))

(fn ConvertLocalRadiusToWorldRadius
  (params (LocalRadius :type Float%20%28double-precision%29))
  (return (* (* LocalRadius 1e+02) (Class|GetClassDefaults (Variables|Default|GetStormShieldClass)))))

(fn GetStormPhase
  ;[0 1 0.6965 0.25] Used by Storm Caller to check what phase this Beacon will server as a center for in Creative Royale.
  (return (Variables|UserOptions|Basic|GetPhase)))

(fn GetFloatOverride
  (params (Value :type Float%20%28double-precision%29) (Default :type Float%20%28double-precision%29))
  (return (select (>= Value 0.0) Value Default)))

(fn GetIntOverride
  (params (Value :type Integer) (Default :type Integer))
  (return (select (>= Value 0) Value Default)))

(fn GetStormPhaseOverride
  (params (In_WaitTime_9_D7E797634847A3AF6330D081A73B4F3D :type Float%20%28single-precision%29) (In_ShrinkTime_11_920D8BB140C25E1B39CCE69746CAB8CB :type Float%20%28single-precision%29) (In_EndRadius_28_FC298B6F485AEC661DE91B9AB8ACDA81 :type Float%20%28single-precision%29) (In_EndRadiusPercentage_3_CD98C461485235F1FA37AEA8E2858699 :type Float%20%28single-precision%29) (In_Damage_38_F609C9DC4F281BA561EB5C8B652CBE3D :type EDamageLevel%20Enum) (In_bMoves_14_1DA1BC714B6A48F21E3E4EABE1D63392 :type Boolean) (In_ForceDistanceMin_22_DB22FD94466C510F2ACB7CB3694C9690 :type Float%20%28single-precision%29) (In_ForceDistanceMax_24_B7CD745E4B355347C2F906989571146F :type Float%20%28single-precision%29) (In_ForceDistanceMinPercentage_33_F185525A4B7DEB217D04F990787CB987 :type Float%20%28single-precision%29) (In_ForceDistanceMaxPercentage_34_34AA9D254BC85E7AF6CBAA8F770AEF4F :type Float%20%28single-precision%29) (In_EndLocation_37_C916E3E84C96C120E2615BAADA29FEE4 :type Vector%204%20Structure))
  (bind (_returnvalue_x _returnvalue_y _returnvalue_z) (Math|Vector|BreakVector (Transformation|GetActorLocation)))
  (return (Utilities|Struct|MakeSAdvancedStormPhaseV2 (CallFunction|GetFloatOverride (Variables|UserOptions|GetWaitTime) In_WaitTime_9_D7E797634847A3AF6330D081A73B4F3D) (CallFunction|GetFloatOverride (Variables|UserOptions|Basic|GetResizeTime) In_ShrinkTime_11_920D8BB140C25E1B39CCE69746CAB8CB) (CallFunction|GetFloatOverride (Variables|Default|GetEndRadiusInCm) In_EndRadius_28_FC298B6F485AEC661DE91B9AB8ACDA81) -1.0 (Variables|UserOptions|Basic|GetDamage%20Level) (Utilities|Enum|Equal(Enum) (Variables|UserOptions|Basic|GetMovementBehavior) "NewEnumerator0") (* (CallFunction|GetFloatOverride (Variables|UserOptions|GetMoveDistanceMin) In_ForceDistanceMin_22_DB22FD94466C510F2ACB7CB3694C9690) 1e+02) (* (CallFunction|GetFloatOverride (Variables|UserOptions|GetMoveDistanceMax) In_ForceDistanceMax_24_B7CD745E4B355347C2F906989571146F) 1e+02) -1.0 -1.0 (select (Utilities|Enum|Equal(Enum) (Variables|UserOptions|Basic|GetMovementBehavior) "NewEnumerator1") (Math|Vector4|MakeVector4 _returnvalue_x _returnvalue_y _returnvalue_z 1.0) In_EndLocation_37_C916E3E84C96C120E2615BAADA29FEE4))))

(fn MigrateLegacyUserOptions
  (bind _damage (Variables|LegacyMigration|GetDamage))
  (bind _legacydamagetodamageleveltable (Variables|LegacyMigration|GetLegacyDamageToDamageLevelTable))
  (bind (_value _returnvalue) (Utilities|Map|Find _legacydamagetodamageleveltable _damage))
  (bind (_damage_level _damage_2) (Class|GetClassDefaults "/CRD_StormController/Device_StormBeaconAdvanced_V2.Device_StormBeaconAdvanced_V2_C"))
  (bind _endradius (Class|GetClassDefaults "/CRD_StormController/Device_StormBeaconAdvanced_V2.Device_StormBeaconAdvanced_V2_C"))
  (bind _forcedistancemin (Class|GetClassDefaults "/CRD_StormController/Device_StormBeaconAdvanced_V2.Device_StormBeaconAdvanced_V2_C"))
  (bind _forcedistancemax (Class|GetClassDefaults "/CRD_StormController/Device_StormBeaconAdvanced_V2.Device_StormBeaconAdvanced_V2_C"))
  (bind _endradius_2 (Variables|LegacyMigration|GetEndRadius))
  (bind _forcedistancemin_2 (Variables|LegacyMigration|GetForceDistanceMin))
  (bind _forcedistancemax_2 (Variables|LegacyMigration|GetForceDistanceMax))
  (|CallParentFunction)
  (Utilities|FlowControl|Sequence
    (:then_0
      (Utilities|FlowControl|Sequence
        (:then_0
          ;[1 0.1681 0.9838 0.25] Convert Legacy EndRadius (cm) to StormRadius (m)
          (if (!= _endradius_2 _endradius)
            (Variables|UserOptions|Basic|SetStormRadius (/ _endradius_2 1e+02))
            (Variables|LegacyMigration|SetEndRadius _endradius)
            (LocalVariables|SetOption%20Migrated true)))
        (:then_1
          ;[1 0.1681 0.9838 0.25] Convert Force Distance Min (cm) to Move Distance Min (m)
          (if (!= _forcedistancemin_2 _forcedistancemin)
            (Variables|UserOptions|SetMoveDistanceMin (/ _forcedistancemin_2 1e+02))
            (Variables|LegacyMigration|SetForceDistanceMin _forcedistancemin)
            (LocalVariables|SetOption%20Migrated true)))
        (:then_2
          ;[1 0.1681 0.9838 0.25] Convert Force Distance Max (cm) to Move Distance Max (m)
          (if (!= _forcedistancemax_2 _forcedistancemax)
            (Variables|UserOptions|SetMoveDistanceMax (/ _forcedistancemax_2 1e+02))
            (Variables|LegacyMigration|SetForceDistanceMax _forcedistancemax)
            (LocalVariables|SetOption%20Migrated true)))
        (:then_3
          ;[1 0.1681 0.9838 0.25] Convert legacy Damage to enum using lookup table
          (if (!= _damage_2 _damage)
            (if _returnvalue
              (Variables|UserOptions|Basic|SetDamage%20Level _value)
              (Variables|LegacyMigration|SetDamage _damage_2)
              (LocalVariables|SetOption%20Migrated true))))))
    (:then_1
      ;[1 0.1681 0.9838 0.25] If any options have migrated, refresh save data from this instance's uproperties.
      (if (LocalVariables|GetOption%20Migrated)
        (OptionData|RefreshPropertyOverrides (Variables|Default|GetToyOptionsComponent))))))

(fn UpdateEndRadiusMeshWorldScale
  (bind (_returnvalue_x _returnvalue_y _returnvalue_z) (Math|Vector|BreakVector (Transformation|GetWorldScale (Variables|Default|GetRadiusPreview))))
  (Transformation|SetWorldScale3D (Variables|Default|GetRadiusPreview) (Math|Vector|MakeVector (CallFunction|ConvertLocalRadiustoWorldRadius (Variables|Default|GetEndRadiusInCm)) (CallFunction|ConvertLocalRadiustoWorldRadius (Variables|Default|GetEndRadiusInCm)) _returnvalue_z)))

(fn ConstructionScript
  (|CallParentFunction))
