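; Advanced storm controller device that spawns and starts a StormShield_Advanced storm with user-set radii, delay and finish behavior.
; Phases come from defaults plus Storm Beacons: custom mode builds the list from insertion-sorted beacons, otherwise beacons override their numbered phase.
; On authority, it switches on Battle Royale safe zones for the minimap, skipping them when the minigame ends.
; It also scales radius preview meshes and migrates legacy centimeter and finish options.
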(event OnVisibilityChanged(VisibleInGameComponent)
  (params (Visible :type Boolean))
  (|CallParentFunction Visible)
  (Audio|Components|Audio|SetVolumeMultiplier (Variables|Default|GetStorm_Manager_Loop_Cue) (select Visible 1.0 0.0) :!class /Script/Engine.AudioComponent))

(event EventOnPropertiesChanged
  (CallFunction|UpdateStartRadiusStaticMeshComponent)
  (CallFunction|UpdateBoundsRadiusStaticMeshComponent))

(event MinigameEndedDelegate(FortMinigameLogic)
  (if (Variables|Default|GetSafeZonesAreActive)
    (CallFunction|EnableSafeZonesforMiniMap true)))

(fn BoundsRadiusInCm
  (return (* (Variables|UserOptions|GetBoundsRadius_m) 1e+02)))

(fn EnableSafeZonesForMiniMap
  (params (DisableSafeZones :type Boolean))
  (bind _returnvalue (BattleRoyaleGamePhaseLogic|GetBattleRoyaleGamePhaseLogicComponent))
  (bind _returnvalue_2 (BattleRoyaleGamePhaseLogic|GetBattleRoyaleGamePhaseLogicComponent))
  (Networking|SwitchHasAuthority
    (:Authority
      (if DisableSafeZones
        (BattleRoyaleGamePhaseLogic|SetSkipSafeZones _returnvalue true)
        (BattleRoyaleGamePhaseLogic|SetGamePhase _returnvalue "None")
        (else
          (Utilities|IsValid _returnvalue_2
            (:Is%20Valid
              (if (Utilities|Enum|Equal(Enum) (BattleRoyaleGamePhaseLogic|GetGamePhase _returnvalue_2) "None")
                (Variables|Default|SetSafeZonesAreActive true)
                (BattleRoyaleGamePhaseLogic|SetSkipSafeZones _returnvalue_2 false)
                (BattleRoyaleGamePhaseLogic|SetGamePhase _returnvalue_2 "SafeZones")
                (BattleRoyaleGamePhaseLogic|JumptoSafeZonePhase _returnvalue_2)))))))))

(fn MigrateLegacyUserOptions
  (bind (_phase_one_radius _boundsradius _onfinishbehavior) (Class|GetClassDefaults "/CRD_StormController/Device_StormControllerAdvanced_V2.Device_StormControllerAdvanced_V2_C"))
  (|CallParentFunction)
  (Utilities|FlowControl|Sequence
    (:then_0
      (Utilities|FlowControl|Sequence
        (:then_0
          ;[1 1 1] Convert Bounds Radius to Meters
          (if (!= _boundsradius (Variables|Legacy|GetBoundsRadius))
            (Variables|UserOptions|SetBoundsRadius_m (/ (Variables|Legacy|GetBoundsRadius) 1e+02))
            (Variables|Legacy|SetBoundsRadius _boundsradius)
            (LocalVariables|SetbMigratedProperties true)))
        (:then_1
          ;[1 1 1] Break up OnFinishBehavior into its enum and float counterparts
          (if (!= _onfinishbehavior (Variables|Legacy|GetOnFinishBehavior))
            (if (>= (Variables|Legacy|GetOnFinishBehavior) 0)
              (Variables|UserOptions|SetOnFinishBehavior_e "NewEnumerator1")
              (Variables|UserOptions|SetDestruction%20Delay (Math|Conversions|ToFloat(Integer) (Variables|Legacy|GetOnFinishBehavior)))
              (else
                (Variables|UserOptions|SetOnFinishBehavior_e "NewEnumerator0")))
            (Variables|Legacy|SetOnFinishBehavior _onfinishbehavior)
            (LocalVariables|SetbMigratedProperties true)))
        (:then_2
          ;[1 1 1] Convert Phase One Radius to Meters
          (if (!= _phase_one_radius (Variables|Legacy|GetPhase%20One%20Radius))
            (Variables|UserOptions|SetPhase%20One%20Radius_m (/ (Variables|Legacy|GetPhase%20One%20Radius) 1e+02))
            (Variables|Legacy|SetPhase%20One%20Radius _phase_one_radius)
            (LocalVariables|SetbMigratedProperties true)))))
    (:then_1
      (if (LocalVariables|GetbMigratedProperties)
        (OptionData|RefreshPropertyOverrides (Variables|Default|GetToyOptionsComponent))))))

(fn OnRep_BoundsRadius_m
  (CallFunction|UpdateBoundsRadiusStaticMeshComponent))

(fn PhaseOneRadiusInCm
  (return (* (Variables|UserOptions|GetPhase%20One%20Radius_m) 1e+02)))

(fn SpawnStormShield
  (bind _stormshield (|CallParentFunction))
  (Utilities|IsValid _stormshield
    (:Is%20Valid
      (bind _asstorm_shield_advanced (Utilities|Casting|CastToStormShield_Advanced _stormshield))
      (Utilities|FlowControl|Sequence
        (:then_0
          (if (Variables|UserOptions|GetUse%20Custom%20Storm%20Phases)
            ;[0 1 0] Handle beacon replacements
            (Utilities|FlowControl|Sequence
              (:then_0
                (bind _foundactors (Utilities|GetActorsRelevanttoMe "/CRD_StormController/Device_StormBeaconAdvanced_V2.Device_StormBeaconAdvanced_V2_C")
                  (:FoundRelevantActors
                    (for _array_element _foundactors
                      (bind _asdevice_storm_beacon_advanced_v2 (Utilities|Casting|CastToDevice_StormBeaconAdvanced_V2 _array_element))
                      ;;[1 0 0] Insert sort into the storm beacons array
                      (if (Utilities|IsArrayNonEmpty (LocalVariables|GetStormBeacons))
                        (bind (_array_element_2 _array_index) (Utilities|Array|ForEachLoopwithBreak (LocalVariables|GetStormBeacons))
                          (:LoopBody
                            (if (> (Variables|/CRD_StormController/Device_StormBeaconAdvanced_V2.Device_StormBeaconAdvanced_V2_C|GetPhase _array_element_2) (Variables|/CRD_StormController/Device_StormBeaconAdvanced_V2.Device_StormBeaconAdvanced_V2_C|GetPhase _asdevice_storm_beacon_advanced_v2))
                              (Utilities|Array|Insert (LocalVariables|GetStormBeacons) _asdevice_storm_beacon_advanced_v2 _array_index)))
                          (:Completed
                            _))
                        (Utilities|Array|AddUnique (LocalVariables|GetStormBeacons) _asdevice_storm_beacon_advanced_v2)
                        (else
                          (Utilities|Array|AddUnique (LocalVariables|GetStormBeacons) _asdevice_storm_beacon_advanced_v2)))))))
              (:then_1
                (for (_array_element_3 _array_index_2) (LocalVariables|GetStormBeacons)
                  (Utilities|Array|SetArrayElem (LocalVariables|GetStormPhases) _array_index_2 (Class|DeviceStormBeaconAdvancedV2|GetStormPhaseOverride _array_element_3 (Utilities|Array|Get(aref) (Variables|Default|GetDefaultStormPhases) (Math|Integer|Min(Integer) _array_index_2 (- (Utilities|Array|Length (Variables|Default|GetDefaultStormPhases)) 1)))) true)))
              (:then_2
                ;;[1 0 0] If no beacons were found, then use the default storm phases
                (if (not (Utilities|IsArrayNonEmpty (LocalVariables|GetStormBeacons)))
                  (LocalVariables|SetStormPhases (Variables|Default|GetDefaultStormPhases)))))
            (else
              ;[0 0 1] Handle beacon overrides
              (Utilities|FlowControl|Sequence
                (:then_0
                  (LocalVariables|SetStormPhases (Variables|Default|GetDefaultStormPhases)))
                (:then_1
                  (bind _foundactors_2 (Utilities|GetActorsRelevanttoMe "/CRD_StormController/Device_StormBeaconAdvanced_V2.Device_StormBeaconAdvanced_V2_C")
                    (:FoundRelevantActors
                      (for _array_element_4 _foundactors_2
                        (bind _asdevice_storm_beacon_advanced_v2_2 (Utilities|Casting|CastToDevice_StormBeaconAdvanced_V2 _array_element_4))
                        (Utilities|Array|AddUnique (LocalVariables|GetStormBeacons) _asdevice_storm_beacon_advanced_v2_2)))))
                (:then_2
                  (for _array_element_5 (LocalVariables|GetStormBeacons)
                    (if (Utilities|Array|IsValidIndex (LocalVariables|GetStormPhases) (- (Variables|/CRD_StormController/Device_StormBeaconAdvanced_V2.Device_StormBeaconAdvanced_V2_C|GetPhase _array_element_5) 1))
                      (Utilities|Array|SetArrayElem (LocalVariables|GetStormPhases) (- (Variables|/CRD_StormController/Device_StormBeaconAdvanced_V2.Device_StormBeaconAdvanced_V2_C|GetPhase _array_element_5) 1) (Class|DeviceStormBeaconAdvancedV2|GetStormPhaseOverride _array_element_5 (Utilities|Array|Get(acopy) (LocalVariables|GetStormPhases) (- (Variables|/CRD_StormController/Device_StormBeaconAdvanced_V2.Device_StormBeaconAdvanced_V2_C|GetPhase _array_element_5) 1)))))))))))
        (:then_1
          (CallFunction|EnableSafeZonesforMiniMap false)
          (Class|StormShieldAdvanced|Start _asstorm_shield_advanced (CallFunction|PhaseOneRadiusinCm) false 0 (LocalVariables|GetStormPhases) (Variables|UserOptions|GetDelayTime) (Utilities|Select :Index (Variables|UserOptions|GetOnFinishBehavior_e) :NewEnumerator0 -1.0 :NewEnumerator1 (Variables|UserOptions|GetDestruction%20Delay)) (CallFunction|BoundsRadiusinCm))
          (return _asstorm_shield_advanced))))))

(fn UpdateBoundsRadiusStaticMeshComponent
  (bind (_returnvalue_x _returnvalue_y _returnvalue_z) (Math|Vector|BreakVector (Transformation|GetWorldScale (Variables|Default|GetEndRadiusVisualization))))
  (Networking|CanExecuteCosmeticEvents
    (:True
      (Transformation|SetWorldScale3D (Variables|Default|GetEndRadiusVisualization) (Math|Vector|MakeVector (CallFunction|ConvertLocalRadiustoWorldRadius (CallFunction|BoundsRadiusinCm)) (CallFunction|ConvertLocalRadiustoWorldRadius (CallFunction|BoundsRadiusinCm)) _returnvalue_z)))))

(fn UpdateStartRadiusStaticMeshComponent
  (bind (_returnvalue_x _returnvalue_y _returnvalue_z) (Math|Vector|BreakVector (Transformation|GetWorldScale (Variables|Default|GetStartRadiusVisualization))))
  (Networking|CanExecuteCosmeticEvents
    (:True
      (Transformation|SetWorldScale3D (Variables|Default|GetStartRadiusVisualization) (Math|Vector|MakeVector (CallFunction|ConvertLocalRadiustoWorldRadius (CallFunction|PhaseOneRadiusinCm)) (CallFunction|ConvertLocalRadiustoWorldRadius (CallFunction|PhaseOneRadiusinCm)) _returnvalue_z)))))

(fn ConstructionScript
  (|CallParentFunction))
