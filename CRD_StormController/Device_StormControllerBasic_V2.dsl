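; Basic storm controller device (V2) that spawns and starts a configurable storm shield that resizes, moves and deals damage.
; SpawnStormShield casts the parent's shield to StormShield_Basic and starts it with user options, converting meters to centimeters and damage level through a map.
; Property changes rescale start and end radius preview meshes; visual updates mute or unmute the loop cue.
; Migration converts legacy centimeter, damage and finish-behavior values, resets them, then refreshes option overrides.
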(event EventBeginPlay
  (Visibility|SetVisibilityComponents (Variables|Default|GetVisibleInGameComponent) (Utilities|Array|MakeArray (Variables|Default|GetStorm_Manager_Loop_Cue)) true)
  (|CallParentFunction))

(event EventUpdateVisuals
  (params (bVisible :type Boolean))
  (|CallParentFunction bVisible)
  (Audio|Components|Audio|SetVolumeMultiplier (Variables|Default|GetStorm_Manager_Loop_Cue) (select bVisible 1.0 0.0) :!class /Script/Engine.AudioComponent))

(event EventOnPropertiesChanged
  (CallFunction|UpdateStartRadiusStaticMeshComponent)
  (CallFunction|UpdateEndRadiusStaticMeshComponent))

(fn GetEndRadiusInCm
  (return (* (Variables|UserOptions|Basic|GetFinalRadius) 1e+02)))

(fn GetInitialRadiusInCm
  (return (* (Variables|UserOptions|Basic|GetInitialRadius) 1e+02)))

(fn MigrateLegacyUserOptions
  (bind _legacydamagetodamageleveltable (Variables|Default|GetLegacyDamageToDamageLevelTable))
  (bind (_value _returnvalue) (Utilities|Map|Find _legacydamagetodamageleveltable (Variables|LegacyMigration|GetDamage)))
  (bind _startradius (Variables|LegacyMigration|GetStartRadius))
  (bind _endradius (Variables|LegacyMigration|GetEndRadius))
  (bind _minmovedistance (Variables|LegacyMigration|GetMinMoveDistance))
  (bind _maxmovedistance (Variables|LegacyMigration|GetMaxMoveDistance))
  (bind _damage (Variables|LegacyMigration|GetDamage))
  (bind _onfinishbehavior (Variables|LegacyMigration|GetOnFinishBehavior))
  (|CallParentFunction)
  (Utilities|FlowControl|Sequence
    (:then_0
      (Utilities|FlowControl|Sequence
        (:then_0
          ;[1 1 1] Convert legacy StartRadius (cm) to InitialRadius (m)
          (if (!= (Class|GetClassDefaults "/CRD_StormController/Device_StormControllerBasic_V2.Device_StormControllerBasic_V2_C") _startradius)
            (Variables|UserOptions|Basic|SetInitialRadius (/ _startradius 1e+02))
            (Variables|LegacyMigration|SetStartRadius (Class|GetClassDefaults "/CRD_StormController/Device_StormControllerBasic_V2.Device_StormControllerBasic_V2_C"))
            (LocalVariables|SetOptions%20Migrated true)))
        (:then_1
          ;[1 1 1] Convert legacy EndRadius (cm) to FinalRadius (m)
          (if (!= (Class|GetClassDefaults "/CRD_StormController/Device_StormControllerBasic_V2.Device_StormControllerBasic_V2_C") _endradius)
            (Variables|UserOptions|Basic|SetFinalRadius (/ _endradius 1e+02))
            (Variables|LegacyMigration|SetEndRadius (Class|GetClassDefaults "/CRD_StormController/Device_StormControllerBasic_V2.Device_StormControllerBasic_V2_C"))
            (LocalVariables|SetOptions%20Migrated true)))
        (:then_2
          ;[1 1 1] Convert legacy MinMoveDistance (cm) to MinimumMoveDistance (m)
          (if (!= (Class|GetClassDefaults "/CRD_StormController/Device_StormControllerBasic_V2.Device_StormControllerBasic_V2_C") _minmovedistance)
            (Variables|UserOptions|SetMinimum%20Move%20Distance (/ _minmovedistance 1e+02))
            (Variables|LegacyMigration|SetMinMoveDistance (Class|GetClassDefaults "/CRD_StormController/Device_StormControllerBasic_V2.Device_StormControllerBasic_V2_C"))
            (LocalVariables|SetOptions%20Migrated true)))
        (:then_3
          ;[1 1 1] Convert legacy MaxMoveDistance (cm) to MaximumMoveDistance (m)
          (if (!= (Class|GetClassDefaults "/CRD_StormController/Device_StormControllerBasic_V2.Device_StormControllerBasic_V2_C") _maxmovedistance)
            (Variables|UserOptions|SetMaximum%20Move%20Distance (/ _maxmovedistance 1e+02))
            (Variables|LegacyMigration|SetMaxMoveDistance (Class|GetClassDefaults "/CRD_StormController/Device_StormControllerBasic_V2.Device_StormControllerBasic_V2_C"))
            (LocalVariables|SetOptions%20Migrated true)))
        (:then_4
          ;[1 1 1] Convert legacy OnFinishBehavior to new split Enum/Float
          (if (!= (Class|GetClassDefaults "/CRD_StormController/Device_StormControllerBasic_V2.Device_StormControllerBasic_V2_C") _onfinishbehavior)
            (if (< _onfinishbehavior 0)
              (Variables|UserOptions|SetFinishBehavior "NewEnumerator0")
              (else
                (Variables|UserOptions|SetFinishBehavior "NewEnumerator1")
                (Variables|UserOptions|SetDestruction%20Delay (Math|Conversions|ToFloat(Integer) _onfinishbehavior))))
            (Variables|LegacyMigration|SetOnFinishBehavior (Class|GetClassDefaults "/CRD_StormController/Device_StormControllerBasic_V2.Device_StormControllerBasic_V2_C"))
            (LocalVariables|SetOptions%20Migrated true)))
        (:then_5
          ;[1 1 1] Convert legacy Damage to enum using lookup table
          (if (!= (Class|GetClassDefaults "/CRD_StormController/Device_StormControllerBasic_V2.Device_StormControllerBasic_V2_C") _damage)
            (if _returnvalue
              (Variables|UserOptions|Basic|SetDamage%20Level _value)
              (Variables|LegacyMigration|SetDamage (Class|GetClassDefaults "/CRD_StormController/Device_StormControllerBasic_V2.Device_StormControllerBasic_V2_C"))
              (LocalVariables|SetOptions%20Migrated true))))))
    (:then_1
      ;[1 1 1] Refresh property overrides if anything was updated during migration
      (if (LocalVariables|GetOptions%20Migrated)
        (OptionData|RefreshPropertyOverrides (Variables|Default|GetToyOptionsComponent))))))

(fn SpawnStormShield
  (bind (_value _returnvalue) (Utilities|Map|Find (Variables|Default|GetDamageToLevelTable) (Variables|UserOptions|Basic|GetDamage%20Level)))
  (bind _stormshield (|CallParentFunction))
  (Utilities|IsValid _stormshield
    (:Is%20Valid
      (bind _asstorm_shield_basic (Utilities|Casting|CastToStormShield_Basic _stormshield))
      (Class|StormShieldBasic|Start _asstorm_shield_basic (Variables|UserOptions|GetDelayTime) (Variables|UserOptions|GetWaitTime) (Variables|UserOptions|GetResize%20Time) (CallFunction|GetInitialRadiusinCm) (CallFunction|GetEndRadiusinCm) (select _returnvalue _value 0) (Utilities|Select :Index (Variables|UserOptions|GetFinishBehavior) :NewEnumerator0 -1.0 :NewEnumerator1 (Variables|UserOptions|GetDestruction%20Delay)) (Variables|UserOptions|GetMoveDirection) (* (Variables|UserOptions|GetMinimum%20Move%20Distance) 1e+02) (* (Variables|UserOptions|GetMaximum%20Move%20Distance) 1e+02) (Variables|UserOptions|GetMoveDelayTime) (Variables|UserOptions|GetMoveTime))
      (return _asstorm_shield_basic))))

(fn UpdateEndRadiusStaticMeshComponent
  (bind (_returnvalue_x _returnvalue_y _returnvalue_z) (Math|Vector|BreakVector (Transformation|GetWorldScale (Variables|Default|GetEndRadiusPreview))))
  (Development|SetHiddeninGame (Variables|Default|GetEndRadiusPreview) (Editor|InEditor) false)
  (Utilities|FlowControl|Sequence
    (:then_0
      (Rendering|SetVisibility (Variables|Default|GetEndRadiusPreview) (and (not (Math|Float|NearlyEqual(Float) (Variables|UserOptions|Basic|GetInitialRadius) (Variables|UserOptions|Basic|GetFinalRadius) 0.01)) (not (Math|Float|NearlyEqual(Float) (Variables|UserOptions|Basic|GetFinalRadius)))) false))
    (:then_1
      (Networking|CanExecuteCosmeticEvents
        (:True
          (Transformation|SetWorldScale3D (Variables|Default|GetEndRadiusPreview) (Math|Vector|MakeVector (CallFunction|ConvertLocalRadiustoWorldRadius (CallFunction|GetEndRadiusinCm)) (CallFunction|ConvertLocalRadiustoWorldRadius (CallFunction|GetEndRadiusinCm)) _returnvalue_z)))))))

(fn UpdateStartRadiusStaticMeshComponent
  (bind (_returnvalue_x _returnvalue_y _returnvalue_z) (Math|Vector|BreakVector (Transformation|GetWorldScale (Variables|Default|GetStartRadiusPreview))))
  (Networking|CanExecuteCosmeticEvents
    (:True
      (Transformation|SetWorldScale3D (Variables|Default|GetStartRadiusPreview) (Math|Vector|MakeVector (CallFunction|ConvertLocalRadiustoWorldRadius (CallFunction|GetInitialRadiusinCm)) (CallFunction|ConvertLocalRadiustoWorldRadius (CallFunction|GetInitialRadiusinCm)) _returnvalue_z))
      (Development|SetHiddeninGame (Variables|Default|GetStartRadiusPreview) (Editor|InEditor) false))))

(fn ConstructionScript
  (|CallParentFunction))
